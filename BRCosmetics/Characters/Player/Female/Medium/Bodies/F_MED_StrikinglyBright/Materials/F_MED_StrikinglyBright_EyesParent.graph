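# Parent eye material for the StrikinglyBright female character body, adding a Slurp effect and time-of-day emissive tint to the shared eye shading.
# MF_eyes_2023 feeds MF_Slurp, with the red channel of the head FX mask as Alpha.
# The resulting emissive is multiplied by Color_Time_of_Day, white from morning to evening and blue-tinted at night.
# The modified material attributes then pass through MF_TechArtOverrides.
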Material /BRCosmetics/Characters/Player/Female/Medium/Bodies/F_MED_StrikinglyBright/Materials/F_MED_StrikinglyBright_EyesParent
graph {
 properties {
  bUseMaterialAttributes=true
  bUsedWithMeshDeformer=true
  bUsedWithSkeletalMesh=true
  pixelDepthOffsetMode=PDOM_Legacy
  shadingModel=MSM_FromMaterialExpression
  usedShadingModels=Default Lit
 }
 nodes [
  n1 = TextureSample() {
   samplerType=SAMPLERTYPE_Masks
   texture=/BRCosmetics/Characters/Player/Female/Medium/Bodies/F_MED_StrikinglyBright/Textures/T_StrikinglyBright_Head_FX
  }
  n2 = GetMaterialAttributes(None=MaterialFunctionCall!(
    Alpha=@n1.R
    AO
    BubblesMask
    SlurpNormalDistortionMask
    Pre Skinned Position
    Material=MaterialFunctionCall() {materialFunction=/Game/Global/Materials/Character_Parent_2023/Eye_Parent_2023/MF_eyes_2023}.Result
   ) {materialFunction=/Game/Weapons/WeaponSkins/MaterialFunction/MF_Slurp}.Result) {attributeGetTypes [#1, #2]}
  n3 = Constant3Vector() {constant=<rgba 1 1 1 0>}
  n4 = MaterialFunctionCall(Color_Morning=@n3, Color_Day=@n3, Color_Evening=@n3, Color_Night=Constant3Vector() {constant=<rgba 0.5 0.5833 1 0>}) {materialFunction=&Color_Time_of_Day}
  MaterialFunctionCall(In=SetMaterialAttributes!(MaterialAttributes=@n2.MaterialAttributes, Base Color=@n2.MaterialAttributes, Emissive Color=Multiply(A=@n2.Emissive Color, B=@n4.Result)) {attributeSetTypes [#1, #2]}) {materialFunction=&MF_TechArtOverrides}
 ]
}
