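# Parent character material for the F_MED_StrikinglyBright body, wrapping a Slurp effect between the shared 2023 character PreFX and PostFX functions.
# A selectable Tech Art Mask channel drives MF_Slurp.
# Emissive blends a faint green masked tint with the slurp glow by PreFX's M red channel, then tints by time of day.
# A zz_NeverEnable static switch can swap base color to T_ThisShouldNotCook, apparently a texture loading/cook test.
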Material /BRCosmetics/Characters/Player/Female/Medium/Bodies/F_MED_StrikinglyBright/Materials/F_MED_StrikinglyBright_Parent
graph {
 properties {
  bUseMaterialAttributes=true
  bUsedWithMeshDeformer=true
  bUsedWithMorphTargets=true
  bUsedWithSkeletalMesh=true
  pixelDepthOffsetMode=PDOM_Legacy
  shadingModel=MSM_FromMaterialExpression
  usedShadingModels=Default Lit | Subsurface | Clear Coat | Subsurface Profile | Cloth
 }
 nodes [
  n1 = MaterialFunctionCall() {materialFunction=&MF_CharacterParent_2023_PreFX}
  n2 = StaticComponentMaskParameter(None=@n1.Tech Art Mask) {defaultR=true, parameterName=Slurp Mask Channel}
  n3 = GetMaterialAttributes(None=MaterialFunctionCall!(Alpha=@n2, AO, BubblesMask, SlurpNormalDistortionMask, Pre Skinned Position, Material=@n1.Result) {materialFunction=/Game/Weapons/WeaponSkins/MaterialFunction/MF_Slurp}.Result) {attributeGetTypes [#1, #2]}
  n4 = Multiply(A=Constant3Vector() {constant=<rgba 0 0.05 0.025 1>}, B=@n2)
  n5 = ComponentMask(None=@n1.M) {r=true}
  n6 = Constant3Vector() {constant=<rgba 1 1 1 0>}
  n7 = Constant3Vector() {constant=<rgba 0.5 0.5833 1 0>}
  n8 = SetMaterialAttributes!(
   MaterialAttributes=@n3.MaterialAttributes
   Base Color=@n3.MaterialAttributes
   Emissive Color=Multiply(
    A=LinearInterpolate(A=@n4, B=@n3.Emissive Color, Alpha=@n5)
    B=MaterialFunctionCall(Color_Morning=@n6, Color_Day=@n6, Color_Evening=@n6, Color_Night=@n7) {materialFunction=&Color_Time_of_Day}.Result
   )
  ) {attributeSetTypes [#1, #2]}
  n9 = TextureSample() {texture=/BRCosmetics/Characters/Player/Female/Medium/Bodies/F_MED_StrikinglyBright/Materials/TextureLoadingTest/T_ThisShouldNotCook}
  MaterialFunctionCall(
   Material=StaticSwitchParameter(
    True=SetMaterialAttributes!(MaterialAttributes=@n8, Base Color=@n9.RGB) {attributeSetTypes [#1]}
    False=@n8
   ) {parameterName=zz_NeverEnable}
   M=@n1.M
   Pre Skinned Local Position=@n1.M
   Distance Blend=@n1.M
   Fresnel=@n1.M
  ) {materialFunction=&MF_CharacterParent_2023_PostFX}
 ]
}
