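# Physics asset for the ponytail on the StrikinglyBright medium female character's face accessory mesh, in the variant named for backpack use.
# Capsules on upper spine, neck, head, clavicles and upper arms form the body shapes around the ponytail, linked by free-rotation constraints.
# Four ponytail capsules chain from a zero-size base on the head with limited swing and tight twist.
# An extra free-rotation constraint links the last segment to the zero-size neck_01 body.
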PhysicsAsset /BRCosmetics/Characters/Player/Female/Medium/Bodies/F_MED_StrikinglyBright/Meshes/Parts/F_MED_StrikinglyBright_FaceAcc_Physics_Backpack
bodies {
 clavicle_l [{shapeType=Capsule, length=6.5317, radius=4.4683, shapeName=clavicle_l_capsule, center=<4.8816 0 0>, rotation=<rot 90 0 0>}]
 clavicle_r [{shapeType=Capsule, length=6.5317, radius=4.4683, shapeName=clavicle_r_capsule, center=<-4.8816 0 0>, rotation=<rot 90 180 0>}]
 dyn_ponytail_01 [{shapeType=Capsule, length=4.081, radius=4.919, shapeName=dyn_ponytail_01_capsule, center=<4.9602 0 2.7891>, rotation=<rot 0 0 90>}]
 dyn_ponytail_02 [{shapeType=Capsule, length=4.081, radius=4.919, shapeName=dyn_ponytail_02_capsule, center=<4.1577 0 0>, rotation=<rot 0 0 90>}]
 dyn_ponytail_03 [{shapeType=Capsule, length=4.6163, radius=4.3837, shapeName=dyn_ponytail_03_capsule, center=<5.1126 0 -0.4491>, rotation=<rot 0 0 90>}]
 dyn_ponytail_04 [{shapeType=Capsule, length=4.081, radius=4.919, shapeName=dyn_ponytail_04_capsule, center=<2.7758 0 -0.3402>, rotation=<rot 0 0 90>}]
 dyn_ponytail_base [{shapeType=Capsule, length=0, radius=0, shapeName=dyn_ponytail_base_capsule}]
 head [{shapeType=Capsule, length=3.3454, radius=8.6546, shapeName=head_capsule, center=<4.8882 -0.2951 0>, rotation=<rot 0 100.8908 90>}]
 neck_01 [{shapeType=Capsule, length=0, radius=0, shapeName=neck_01_capsule}]
 neck_02 [{shapeType=Capsule, length=4.719, radius=4.281, shapeName=neck_02_capsule, center=<-1.5212 0 0>, rotation=<rot 90 0 0>}]
 spine_05 [{shapeType=Capsule, length=6.9474, radius=10.2156, shapeName=spine_05_capsule}]
 upperarm_l [{shapeType=Capsule, length=21.318, radius=4.5597, shapeName=upperarm_l_capsule, center=<11.8544 0 0>, rotation=<rot 90 0 0>}]
 upperarm_r [{shapeType=Capsule, length=21.318, radius=4.5597, shapeName=upperarm_r_capsule, center=<-11.8544 0 0>, rotation=<rot 90 180 0>}]
}
constraints [
 {
  bone1Name=clavicle_l
  bone2Name=spine_05
  swing1LimitDegrees=45
  swing1Motion=Free
  swing2LimitDegrees=45
  swing2Motion=Free
  twistLimitDegrees=45
  twistMotion=Free
 }
 {
  bone1Name=upperarm_l
  bone2Name=clavicle_l
  swing1LimitDegrees=45
  swing1Motion=Free
  swing2LimitDegrees=45
  swing2Motion=Free
  twistLimitDegrees=45
  twistMotion=Free
 }
 {
  bone1Name=clavicle_r
  bone2Name=spine_05
  swing1LimitDegrees=45
  swing1Motion=Free
  swing2LimitDegrees=45
  swing2Motion=Free
  twistLimitDegrees=45
  twistMotion=Free
 }
 {
  bone1Name=upperarm_r
  bone2Name=clavicle_r
  swing1LimitDegrees=45
  swing1Motion=Free
  swing2LimitDegrees=45
  swing2Motion=Free
  twistLimitDegrees=45
  twistMotion=Free
 }
 {
  bone1Name=neck_01
  bone2Name=spine_05
  swing1LimitDegrees=45
  swing1Motion=Free
  swing2LimitDegrees=45
  swing2Motion=Free
  twistLimitDegrees=45
  twistMotion=Free
 }
 {
  bone1Name=neck_02
  bone2Name=neck_01
  swing1LimitDegrees=45
  swing1Motion=Free
  swing2LimitDegrees=45
  swing2Motion=Free
  twistLimitDegrees=45
  twistMotion=Free
 }
 {
  bone1Name=head
  bone2Name=neck_02
  swing1LimitDegrees=45
  swing1Motion=Free
  swing2LimitDegrees=45
  swing2Motion=Free
  twistLimitDegrees=45
  twistMotion=Free
 }
 {
  bone1Name=dyn_ponytail_base
  bone2Name=head
  swing1LimitDegrees=45
  swing1Motion=Free
  swing2LimitDegrees=45
  swing2Motion=Free
  twistLimitDegrees=45
  twistMotion=Free
 }
 {bone1Name=dyn_ponytail_01, bone2Name=dyn_ponytail_base, swing1LimitDegrees=25, swing2LimitDegrees=25, twistLimitDegrees=2}
 {bone1Name=dyn_ponytail_02, bone2Name=dyn_ponytail_01, swing1LimitDegrees=25, swing2LimitDegrees=45, twistLimitDegrees=2}
 {bone1Name=dyn_ponytail_03, bone2Name=dyn_ponytail_02, swing1LimitDegrees=25, swing2LimitDegrees=45, twistLimitDegrees=2}
 {bone1Name=dyn_ponytail_04, bone2Name=dyn_ponytail_03, swing1LimitDegrees=25, swing2LimitDegrees=45, twistLimitDegrees=2}
 {
  bone1Name=dyn_ponytail_04
  bone2Name=neck_01
  swing1LimitDegrees=45
  swing1Motion=Free
  swing2LimitDegrees=45
  swing2Motion=Free
  twistLimitDegrees=45
  twistMotion=Free
 }
]
