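# Post-copy animation Blueprint for the StrikinglyBright female FaceAcc character part, adding secondary physics motion to its hair and ponytail.
# It copies the attached parent's pose, layering in the Body mesh's pose from the clavicles.
# Head-space RigidBody simulations switch physics assets on an Is Backpack check, then feed chained AnimDynamics linear springs for front and side hair, then ponytail bones.
# The event graph caches the Body mesh and updates backpack and gravity-override state.
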AnimBlueprint /BRCosmetics/Characters/Player/Female/Medium/Bodies/F_MED_StrikinglyBright/Meshes/Parts/F_MED_StrikinglyBright_FaceAcc_AnimBP
target_skeleton=&/F_MED_StrikinglyBright_FaceAcc_Skeleton
graphs [
 AnimGraph {nodes [
   n1 = Anim Dynamics - Bone: dyn_ponytail_02(ComponentPose=Local To Component(LocalPose=Use cached pose 'toAnimDynPony'()), Alpha=1, ExternalForce="0,0,0") {
    angularSpringConstant=100
    bLinearSpring=true
    bOverrideLinearDamping=true
    bUseGravityOverride=true
    linearDampingOverride=1
    linearSpringConstant=250
    simulationSpace=RootRelative
    alphaBoolBlend {blendOption=Linear}
    boundBone {boneName=dyn_ponytail_02}
    componentAppliedLinearAccClamp=<3000 3000 3000>
    gravityOverride {z=-980}
    physicsBodyDefinitions [{
      boundBone {boneName=dyn_ponytail_02}
      boxExtents=<20 20 20>
      constraintSetup {angularLimitsMax=<0 0 0>, angularLimitsMin=<0 0 0>, angularTarget=<0 0 0>, linearAxesMax=<0 2 0>, linearAxesMin=<-3 -2 0>}
      localJointOffset=<0.1 0 0>
     }]
   }
   n2 = Anim Dynamics - Bone: dyn_ponytail_03(ComponentPose=@n1, Alpha=1, ExternalForce="0,0,0") {
    angularSpringConstant=100
    bLinearSpring=true
    bOverrideLinearDamping=true
    bUseGravityOverride=true
    linearDampingOverride=1
    linearSpringConstant=250
    simulationSpace=RootRelative
    alphaBoolBlend {blendOption=Linear}
    boundBone {boneName=dyn_ponytail_03}
    componentAppliedLinearAccClamp=<3000 3000 3000>
    gravityOverride {z=-980}
    physicsBodyDefinitions [{
      boundBone {boneName=dyn_ponytail_03}
      boxExtents=<20 20 20>
      constraintSetup {angularLimitsMax=<0 0 0>, angularLimitsMin=<0 0 0>, angularTarget=<0 0 0>, linearAxesMax=<0 2 0>, linearAxesMin=<-3 -2 0>}
      localJointOffset=<0.1 0 0>
     }]
   }
   n3 = Anim Dynamics - Bone: dyn_ponytail_04(ComponentPose=@n2, Alpha=1, ExternalForce="0,0,0") {
    angularSpringConstant=100
    bLinearSpring=true
    bOverrideLinearDamping=true
    bUseGravityOverride=true
    linearDampingOverride=1
    linearSpringConstant=250
    simulationSpace=RootRelative
    alphaBoolBlend {blendOption=Linear}
    boundBone {boneName=dyn_ponytail_04}
    componentAppliedLinearAccClamp=<3000 3000 3000>
    gravityOverride {z=-980}
    physicsBodyDefinitions [{
      boundBone {boneName=dyn_ponytail_04}
      boxExtents=<20 20 20>
      constraintSetup {angularLimitsMax=<0 0 0>, angularLimitsMin=<0 0 0>, angularTarget=<0 0 0>, linearAxesMax=<0 2 0>, linearAxesMin=<-3 -2 0>}
      localJointOffset=<0.1 0 0>
     }]
   }
   Output Pose(Result=Component To Local(ComponentPose=@n3)) {layerGroup=None}
   n4 = RigidBody!(ComponentPose=Local To Component(LocalPose=Use cached pose 'ToPhysics'()), Alpha, OverrideWorldGravity:Vector="0,0,0", ExternalForce) -> (Pose) {
    bAllowPoseScalePropagation=false
    bClampLinearTranslationLimitToRefPose=true
    bOverrideWorldGravity=true
    overridePhysicsAsset=&/F_MED_StrikinglyBright_FaceAcc_Physics_Backpack
    simulationSpace=BaseBoneSpace
    alphaBoolBlend {blendOption=Linear}
    baseBoneRef {boneName=head}
    componentAppliedLinearAccClamp=<3000 3000 3000>
    simSpaceSettings {
     maxAngularAcceleration=100
     maxAngularVelocity=6
     maxLinearAcceleration=150
     maxLinearVelocity=100
     velocityScaleZ=1200
     worldAlpha=1
     externalLinearDragV=<1 1 1>
    }
   }
   n5 = RigidBody(ComponentPose=Local To Component(LocalPose=Use cached pose 'ToPhysics'()), OverrideWorldGravity="0,0,0") {
    bClampLinearTranslationLimitToRefPose=true
    bOverrideWorldGravity=true
    overridePhysicsAsset=&/F_MED_StrikinglyBright_FaceAcc_Physics
    simulationSpace=BaseBoneSpace
    alphaBoolBlend {blendOption=Linear}
    baseBoneRef {boneName=head}
    componentAppliedLinearAccClamp=<3000 3000 3000>
    simSpaceSettings {
     maxAngularAcceleration=100
     maxAngularVelocity=6
     maxLinearAcceleration=150
     maxLinearVelocity=100
     velocityScaleZ=1200
     worldAlpha=1
     externalLinearDragV=<1 1 1>
    }
   }
   Save cached pose 'toPonytailDyn'(Pose=Blend Poses by bool(
     BlendPose_0=Component To Local(ComponentPose=@n4)
     BlendPose_1=Component To Local(ComponentPose=@n5)
     bActiveValue=Get Is Backpack() -> (Is Backpack:bool)
    ) {blendType=Linear}) {cacheName=toPonytailDyn}
   n6 = Anim Dynamics - Bone: dyn_hair_frnt_r_01(ComponentPose=Local To Component(LocalPose=Use cached pose 'toPonytailDyn'()), Alpha=0.5, ExternalForce="0,0,0") {
    bLinearSpring=true
    bOverrideLinearDamping=true
    bUseGravityOverride=true
    linearDampingOverride=1
    linearSpringConstant=300
    simulationSpace=RootRelative
    alphaBoolBlend {blendOption=Linear}
    boundBone {boneName=dyn_hair_frnt_r_01}
    componentAppliedLinearAccClamp=<3000 3000 3000>
    gravityOverride {z=-980}
    physicsBodyDefinitions [{
      boundBone {boneName=dyn_hair_frnt_r_01}
      boxExtents=<5 5 5>
      constraintSetup {angularLimitsMax=<0 0 0>, angularLimitsMin=<0 0 0>, angularTarget=<0 0 0>, linearAxesMax=<0 0 2>, linearAxesMin=<0 0 -2>}
      localJointOffset=<0.1 0 0>
     }]
   }
   n7 = Anim Dynamics - Bone: dyn_hair_frnt_r_02(ComponentPose=@n6, Alpha=0.5, ExternalForce="0,0,0") {
    bLinearSpring=true
    bOverrideLinearDamping=true
    bUseGravityOverride=true
    linearDampingOverride=1
    linearSpringConstant=300
    simulationSpace=RootRelative
    alphaBoolBlend {blendOption=Linear}
    boundBone {boneName=dyn_hair_frnt_r_02}
    componentAppliedLinearAccClamp=<3000 3000 3000>
    gravityOverride {z=-980}
    physicsBodyDefinitions [{
      boundBone {boneName=dyn_hair_frnt_r_02}
      boxExtents=<5 5 5>
      constraintSetup {angularLimitsMax=<0 0 0>, angularLimitsMin=<0 0 0>, angularTarget=<0 0 0>, linearAxesMax=<0 0 2>, linearAxesMin=<0 0 -2>}
      localJointOffset=<0.1 0 0>
     }]
   }
   n8 = Anim Dynamics - Bone: dyn_hair_frnt_r_03(ComponentPose=@n7, Alpha=0.5, ExternalForce="0,0,0") {
    bLinearSpring=true
    bOverrideLinearDamping=true
    bUseGravityOverride=true
    linearDampingOverride=1
    linearSpringConstant=300
    simulationSpace=RootRelative
    alphaBoolBlend {blendOption=Linear}
    boundBone {boneName=dyn_hair_frnt_r_03}
    componentAppliedLinearAccClamp=<3000 3000 3000>
    gravityOverride {z=-980}
    physicsBodyDefinitions [{
      boundBone {boneName=dyn_hair_frnt_r_03}
      boxExtents=<5 5 5>
      constraintSetup {angularLimitsMax=<0 0 0>, angularLimitsMin=<0 0 0>, angularTarget=<0 0 0>, linearAxesMax=<0 0 2>, linearAxesMin=<0 0 -2>}
      localJointOffset=<0.1 0 0>
     }]
   }
   n9 = Anim Dynamics - Bone: dyn_hair_frnt_l(ComponentPose=@n8, Alpha=0.5, ExternalForce="0,0,0") {
    bLinearSpring=true
    bOverrideLinearDamping=true
    bUseGravityOverride=true
    linearDampingOverride=0.9999
    linearSpringConstant=300
    simulationSpace=RootRelative
    alphaBoolBlend {blendOption=Linear}
    boundBone {boneName=dyn_hair_frnt_l}
    componentAppliedLinearAccClamp=<3000 3000 3000>
    gravityOverride {z=-980}
    physicsBodyDefinitions [{
      boundBone {boneName=dyn_hair_frnt_l}
      boxExtents=<10 10 10>
      constraintSetup {angularLimitsMax=<0 0 0>, angularLimitsMin=<0 0 0>, angularTarget=<0 0 0>, linearAxesMax=<0 0 2>, linearAxesMin=<0 0 -2>}
      localJointOffset=<0.1 0 0>
     }]
   }
   n10 = Anim Dynamics - Bone: dyn_hair_side_r(ComponentPose=@n9, Alpha=0.5, ExternalForce="0,0,0") {
    bLinearSpring=true
    bOverrideLinearDamping=true
    bUseGravityOverride=true
    linearDampingOverride=0.9999
    linearSpringConstant=300
    simulationSpace=RootRelative
    alphaBoolBlend {blendOption=Linear}
    boundBone {boneName=dyn_hair_side_r}
    componentAppliedLinearAccClamp=<3000 3000 3000>
    gravityOverride {z=-980}
    physicsBodyDefinitions [{
      boundBone {boneName=dyn_hair_side_r}
      boxExtents=<10 10 10>
      constraintSetup {angularLimitsMax=<0 0 0>, angularLimitsMin=<0 0 0>, angularTarget=<0 0 0>, linearAxesMax=<0 0 2>, linearAxesMin=<0 0 -2>}
      localJointOffset=<0.1 0 0>
     }]
   }
   n11 = Anim Dynamics - Bone: dyn_hair_side_l(ComponentPose=@n10, Alpha=0.5, ExternalForce="0,0,0") {
    bLinearSpring=true
    bOverrideLinearDamping=true
    bUseGravityOverride=true
    linearDampingOverride=0.9999
    linearSpringConstant=300
    simulationSpace=RootRelative
    alphaBoolBlend {blendOption=Linear}
    boundBone {boneName=dyn_hair_side_l}
    componentAppliedLinearAccClamp=<3000 3000 3000>
    gravityOverride {z=-980}
    physicsBodyDefinitions [{
      boundBone {boneName=dyn_hair_side_l}
      boxExtents=<10 10 10>
      constraintSetup {angularLimitsMax=<0 0 0>, angularLimitsMin=<0 0 0>, angularTarget=<0 0 0>, linearAxesMax=<0 0 2>, linearAxesMin=<0 0 -2>}
      localJointOffset=<0.1 0 0>
     }]
   }
   Save cached pose 'toAnimDynPony'(Pose=Component To Local(ComponentPose=@n11)) {cacheName=toAnimDynPony}
   n12 = Get Body() -> (Body:Skeletal Mesh Component Budgeted Object Reference)
   Save cached pose 'ToPhysics'(Pose=Layered blend per bone(
     BasePose=Copy Pose From Mesh() {bUseAttachedParent=true}
     BlendPoses_0=Copy Pose From Mesh(SourceMeshComponent=@n12) {bUseAttachedParent=true}
    ) {layerSetup [{branchFilters [{blendDepth=3, boneName=clavicle_l}, {blendDepth=3, boneName=clavicle_r}]}]}) {cacheName=ToPhysics}
  ]}
 EventGraph {nodes [
   Event BlueprintUpdateAnimation()
   > n1 = Gravity Override State Function DEPRECATED(CharacterPartAnimInstance=Self-Reference(), GameTimeInSeconds=GetGameTimeInSeconds(), GravityOverrideParams=Get Gravity Override Params())
   > Set Gravity Override(Gravity Override=@n1)
   > Set Is Backpack(Is Backpack:bool=CheckForBackbling()) -> (Output_Get:bool)
   n2 = Event OnCharacterCustomizationLoaded()
   > Set Body(Body:Skeletal Mesh Component Budgeted Object Reference=Get Array Item with EFort Custom Part Type(Array=@n2.CustomMeshParts)) -> (Output_Get:Skeletal Mesh Component Budgeted Object Reference)
  ]}
]
variables [
 Body {type=SkeletalMeshComponentBudgeted}
 Gravity Override {type=Vector}
 Gravity Override Params {type=GravityOverrideParamsStruct}
 Is Backpack {type=boolean}
]
